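; Weapon Blueprint for the FirePetal Flavor shotgun, adding cosmetic effect and mesh tweaks on top of its parent class.
; Muzzle flashes alternate between two barrel positions, using magazine ammo parity to set BarrelMuzzleIndex and offset the spawned muzzle system.
; Impact effects get a fixed DebrisAmount parameter after the parent spawns them.
; Adding a non-default optic mod reveals the rail through a dynamic material scalar on the weapon mesh.
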(event EventTick
  (params (DeltaSeconds :type Float%20%28single-precision%29)))

(event FortWeapon|EventOnPlayImpactFX
  (params (HitResult :type Hit%20Result%20Structure%20%28by%20ref%29) (ImpactPhysicalSurface :type EPhysicalSurface%20Enum :default "SurfaceType_Default") (SpawnedPSC :type FXSystem%20Component%20Object%20Reference))
  (|CallParentFunction HitResult ImpactPhysicalSurface SpawnedPSC)
  (Effects|Components|ParticleSystem|SetFloatParameter SpawnedPSC "DebrisAmount" 0.1))

(event EventMuzzleFlashFX
  (params (Persistent%20Fire :type Boolean))
  (|CallParentFunction Persistent%20Fire)
  (if (== (Math|Integer|%25%28Integer%29 (FortWeapon|GetLocalMagazineAmmoCount) 2) 0)
    (Variables|Default|SetBarrelMuzzleIndex 0)
    (else
      (Variables|Default|SetBarrelMuzzleIndex 1)))
  (Transformation|SetRelativeLocation (Variables|Effects|Muzzle|GetSpawned%20Muzzle%20System) (Utilities|Select :Index (Variables|Default|GetBarrelMuzzleIndex) :Option%200 "0, 0, 0" :Option%201 "0.000000,3.900000,0.000000") false false))

(event EventBeginPlay
  (|CallParentFunction)
  (WeaponMods|AssignOnModAddedtoWeapon (delegate Custom|OnModAddedToWeapon_Event)))

(event Custom|OnModAddedToWeapon_Event
  (params (Mod :type Fort%20Weapon%20Mod%20Item%20Definition%20Object%20Reference) (Weapon :type Fort%20Weapon%20Object%20Reference))
  ;[1 1 1] When an Optic Mod is applied to the weapon, display the Rail
  (if (and (!= Mod "/WeaponMods/Items/WeaponMods/Optics/IronSight/WMID_Default_IronSights.WMID_Default_IronSights") (== (Variables|/Script/FortniteGame.FortWeaponModItemDefinition|GetModSlot Mod) (Variables|Default|GetOpticModTag)))
    (bind _returnvalue (Fortnite|GetDynamicMaterialInstance (Variables|Visual|GetWeaponMesh) 1))
    (Rendering|Material|SetScalarParameterValue _returnvalue "Hide Element 01" 0.0 :!class /Script/Engine.MaterialInstanceDynamic)))

(fn ConstructionScript
  (|CallParentFunction))
